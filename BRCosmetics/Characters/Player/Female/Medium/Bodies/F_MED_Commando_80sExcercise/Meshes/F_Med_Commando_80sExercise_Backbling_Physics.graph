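# Physics asset for the Backbling mesh of the female medium Commando 80s Exercise cosmetic, giving its shirt simple physical motion.
# Front, back, left and right dyn_shirt bones each carry a thin box body, constrained directly to spine_02 with swing limits and a tighter twist limit.
# Capsules on spine_02 and spine_03 form the torso volume, joined by a constraint with free swing and twist.
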PhysicsAsset /BRCosmetics/Characters/Player/Female/Medium/Bodies/F_MED_Commando_80sExcercise/Meshes/F_Med_Commando_80sExercise_Backbling_Physics
bodies {
 dyn_shirt_b [{shapeType=Box, extentX=16.891, extentY=2, extentZ=22.4913, center=<3.9867 0 0>}]
 dyn_shirt_f [{shapeType=Box, extentX=10, extentY=2.2206, extentZ=19.1644, center=<3 0.3415 0>}]
 dyn_shirt_l [{shapeType=Box, extentX=10, extentY=2, extentZ=17.7685, center=<3 0 0>}]
 dyn_shirt_r [{shapeType=Box, extentX=10, extentY=2, extentZ=20.097, center=<3 0 0>}]
 spine_02 [{shapeType=Capsule, length=3.5402, radius=8.6983, center=<0 1.1455 0.0001>}]
 spine_03 [{shapeType=Capsule, length=4.2982, radius=8.6762, center=<0 0.6937 0>}]
}
constraints [
 {
  bone1Name=spine_03
  bone2Name=spine_02
  swing1LimitDegrees=45
  swing1Motion=Free
  swing2LimitDegrees=45
  swing2Motion=Free
  twistLimitDegrees=45
  twistMotion=Free
 }
 {bone1Name=dyn_shirt_r, bone2Name=spine_02, swing1LimitDegrees=45, swing2LimitDegrees=45, twistLimitDegrees=15}
 {bone1Name=dyn_shirt_b, bone2Name=spine_02, swing1LimitDegrees=45, swing2LimitDegrees=45, twistLimitDegrees=15}
 {bone1Name=dyn_shirt_l, bone2Name=spine_02, swing1LimitDegrees=45, swing2LimitDegrees=45, twistLimitDegrees=15}
 {bone1Name=dyn_shirt_f, bone2Name=spine_02, swing1LimitDegrees=45, swing2LimitDegrees=45, twistLimitDegrees=15}
]
