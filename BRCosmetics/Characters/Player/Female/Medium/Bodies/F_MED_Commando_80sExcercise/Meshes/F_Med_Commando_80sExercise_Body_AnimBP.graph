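# Post-copy body AnimBP for the female 80s Exercise Commando cosmetic, adding secondary physics to a pose copied from the attached parent mesh.
# Each branch runs necklace AnimDynamics, then a spine_03-based RigidBody node, selected by Is Wearing Backbling, which updates when customization loads.
# The backbling branch swaps in its own physics asset and additively rotates the back shirt bone.
# A speed-weighted shirt AnimDynamics chain exists but does not reach the output pose.
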AnimBlueprint /BRCosmetics/Characters/Player/Female/Medium/Bodies/F_MED_Commando_80sExcercise/Meshes/F_Med_Commando_80sExercise_Body_AnimBP
target_skeleton=&/F_Med_Commando_80sExercise_Skeleton
graphs [
 AnimGraph {nodes [
   n1 = MapRangeClamped(Value=Get PawnSpeed2D(), InRangeA=0, InRangeB=100, OutRangeA=0.75, OutRangeB=0.25)
   n2 = Anim Dynamics - Bone: dyn_shirt_f(Alpha=@n1, ExternalForce="0,0,0") {
    bGravityOverrideInSimSpace=true
    bLinearSpring=true
    bOverrideLinearDamping=true
    gravityScale=0.05
    linearDampingOverride=0.99
    linearSpringConstant=0.75
    numSolverIterationsPostUpdate=4
    numSolverIterationsPreUpdate=16
    alphaBoolBlend {blendOption=Linear}
    boundBone {boneName=dyn_shirt_f}
    componentLinearVelScale=<1 1 1>
    physicsBodyDefinitions [{
      sphereCollisionRadius=0
      boundBone {boneName=dyn_shirt_f}
      boxExtents=<3 3 3>
      constraintSetup {angularLimitsMax=<0 0 0>, angularLimitsMin=<0 0 0>, angularTarget=<0 0 0>, linearAxesMax=<1 1 1>, linearAxesMin=<-1 0 -1>}
      localJointOffset=<0 0 0.1>
     }]
   }
   n3 = Anim Dynamics - Bone: dyn_shirt_l(ComponentPose=@n2, Alpha=@n1, ExternalForce="0,0,0") {
    bGravityOverrideInSimSpace=true
    bLinearSpring=true
    bOverrideLinearDamping=true
    gravityScale=0.05
    linearDampingOverride=0.99
    linearSpringConstant=100
    numSolverIterationsPostUpdate=4
    numSolverIterationsPreUpdate=16
    alphaBoolBlend {blendOption=Linear}
    boundBone {boneName=dyn_shirt_l}
    componentLinearVelScale=<1 1 1>
    physicsBodyDefinitions [{
      sphereCollisionRadius=0
      boundBone {boneName=dyn_shirt_l}
      boxExtents=<3 3 3>
      constraintSetup {angularLimitsMax=<0 0 0>, angularLimitsMin=<0 0 0>, angularTarget=<0 0 0>, linearAxesMax=<1 2 1>, linearAxesMin=<-0.5 -2 -1>}
      localJointOffset=<-0.1 0 0.1>
     }]
   }
   n4 = Anim Dynamics - Bone: dyn_shirt_b(ComponentPose=@n3, Alpha=@n1, ExternalForce="0,0,0") {
    bGravityOverrideInSimSpace=true
    bLinearSpring=true
    bOverrideLinearDamping=true
    gravityScale=0.03
    linearDampingOverride=0.99
    linearSpringConstant=85
    numSolverIterationsPostUpdate=4
    numSolverIterationsPreUpdate=16
    simulationSpace=BoneRelative
    alphaBoolBlend {blendOption=Linear}
    boundBone {boneName=dyn_shirt_b}
    componentLinearVelScale=<1 1 1>
    physicsBodyDefinitions [{
      sphereCollisionRadius=0
      boundBone {boneName=dyn_shirt_b}
      boxExtents=<3 3 3>
      constraintSetup {angularLimitsMax=<0 0 0>, angularLimitsMin=<0 0 0>, angularTarget=<0 0 0>, linearAxesMax=<1 0 1>, linearAxesMin=<-1 -1 -1>}
      localJointOffset=<0 0.1 0.1>
     }]
    relativeSpaceBone {boneName=root}
   }
   Anim Dynamics - Bone: dyn_shirt_r(ComponentPose=@n4, Alpha=@n1, ExternalForce="0,0,0") {
    bGravityOverrideInSimSpace=true
    bLinearSpring=true
    bOverrideLinearDamping=true
    gravityScale=0.07
    linearDampingOverride=0.99
    linearSpringConstant=80
    numSolverIterationsPostUpdate=4
    numSolverIterationsPreUpdate=16
    alphaBoolBlend {blendOption=Linear}
    boundBone {boneName=dyn_shirt_r}
    componentLinearAccScale {x=-1}
    componentLinearVelScale=<1 1 1>
    physicsBodyDefinitions [{
      sphereCollisionRadius=0
      boundBone {boneName=dyn_shirt_r}
      boxExtents=<3 3 3>
      constraintSetup {angularLimitsMax=<0 0 0>, angularLimitsMin=<0 0 0>, angularTarget=<0 0 0>, linearAxesMax=<0 1 1>, linearAxesMin=<-1 -1 -1>}
      localJointOffset=<0.1 0 0.1>
     }]
   }
   n5 = Anim Dynamics - Bone: dyn_necklace(ComponentPose=Local To Component(LocalPose=Copy Pose From Mesh() {bUseAttachedParent=true}), Alpha=1, ExternalForce="0,0,0") {
    angularDampingOverride=0.98
    bGravityOverrideInSimSpace=true
    bOverrideAngularDamping=true
    gravityScale=0.75
    numSolverIterationsPostUpdate=4
    numSolverIterationsPreUpdate=16
    alphaBoolBlend {blendOption=Linear}
    boundBone {boneName=dyn_necklace}
    componentLinearAccScale=<0.5 0.5 0.5>
    physicsBodyDefinitions [{
      sphereCollisionRadius=0
      boundBone {boneName=dyn_necklace}
      boxExtents=<3 3 3>
      constraintSetup {angularLimitsMax=<10 0 10>, angularLimitsMin=<-10 0 -10>, angularTarget=<1 0 0>}
      localJointOffset=<0 0.1 -0.3>
     }]
   }
   n6 = RigidBody(ComponentPose=@n5) {
    bOverrideWorldGravity=true
    overridePhysicsAsset=&/F_Med_Commando_80sExercise_Body_Physics
    simulationSpace=BaseBoneSpace
    alphaBoolBlend {blendOption=Linear}
    baseBoneRef {boneName=spine_03}
    componentLinearAccScale=<0.5 0.5 0.5>
    overrideWorldGravity {z=-980}
    simSpaceSettings {velocityScaleZ=0.1, worldAlpha=1}
   }
   n7 = Anim Dynamics - Bone: dyn_necklace(ComponentPose=Local To Component(LocalPose=Copy Pose From Mesh() {bUseAttachedParent=true}), Alpha=1, ExternalForce="0,0,0") {
    angularDampingOverride=0.98
    bGravityOverrideInSimSpace=true
    bOverrideAngularDamping=true
    gravityScale=0.75
    numSolverIterationsPostUpdate=4
    numSolverIterationsPreUpdate=16
    alphaBoolBlend {blendOption=Linear}
    boundBone {boneName=dyn_necklace}
    componentLinearAccScale=<0.5 0.5 0.5>
    physicsBodyDefinitions [{
      sphereCollisionRadius=0
      boundBone {boneName=dyn_necklace}
      boxExtents=<3 3 3>
      constraintSetup {angularLimitsMax=<10 0 10>, angularLimitsMin=<-10 0 -10>, angularTarget=<1 0 0>}
      localJointOffset=<0 0.1 -0.3>
     }]
   }
   n8 = RigidBody(ComponentPose=@n7) {
    bOverrideWorldGravity=true
    overridePhysicsAsset=&/F_Med_Commando_80sExercise_Backbling_Physics
    simulationSpace=BaseBoneSpace
    alphaBoolBlend {blendOption=Linear}
    baseBoneRef {boneName=spine_03}
    componentLinearAccScale=<0.5 0.5 0.5>
    overrideWorldGravity {z=-980}
    simSpaceSettings {velocityScaleZ=0.1, worldAlpha=1}
   }
   n9 = "Transform (Modify) Bone - Bone: dyn_shirt_b"(ComponentPose=@n8, Alpha=1, Translation="0,0,0", Rotation="0,10,0", Scale="1,1,1") {
    rotationMode=BMM_Additive
    rotationSpace=BCS_BoneSpace
    translationMode=BMM_Additive
    translationSpace=BCS_BoneSpace
    alphaBoolBlend {blendOption=Linear}
    boneToModify {boneName=dyn_shirt_b}
    rotation {yaw=10}
   }
   Output Pose(Result=Blend Poses by bool(
     BlendPose_0=Component To Local(ComponentPose=@n9)
     BlendPose_1=Component To Local(ComponentPose=@n6)
     bActiveValue=Get Is Wearing Backbling()
    ) {blendType=Linear}) {layerGroup=None}
  ]}
 EventGraph {nodes [
   Delay(Duration=0.1)
   > Do Once()
   Event OnCharacterCustomizationLoaded()
   > n1 = Sequence!(execute) -> (then_0, then_1, then_2:Exec)
   > Cast To PlayerPawn_Athena(Object=TryGetPawnOwner())
   n2 = CheckForBackbling()
   Set Is Wearing Backbling(execute=@n1.then_1, Is Wearing Backbling=OR Boolean(A=@n2, B=@n2.bIsBackblingLong))
   Event BlueprintInitializeAnimation()
   > Set MeshToCopy(MeshToCopy:Skeletal Mesh Component Object Reference=GetOwnerSkeletalMeshComponent()) -> (Output_Get:Skeletal Mesh Component Object Reference)
   TryGetPawnOwner()
   Event BlueprintUpdateAnimation()
  ]}
]
variables [
 Is Wearing Backbling {type=boolean}
 MeshToCopy {type=SkeletalMeshComponent}
]
